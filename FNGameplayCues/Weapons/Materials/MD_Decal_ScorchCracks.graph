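# Translucent deferred decal material for scorch marks with glowing cracks, likely left by weapon impacts.
# A crack texture, optionally sphere-masked, is multiplied by a char mask and tinted between edge and center colors for HDR emissive.
# Decal lifetime drives the fade, with a rising power exponent thinning the glow before the char disappears.
# Opacity uses the char texture, plus an optional world-normal projection fade that Forward and Mobile paths skip.
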Material /FNGameplayCues/Weapons/Materials/MD_Decal_ScorchCracks
graph {
 properties {
  bAutomaticallySetUsageInEditor=false
  blendMode=BLEND_Translucent
  materialDomain=MD_DeferredDecal
  pixelDepthOffsetMode=PDOM_Legacy
  shadingModel=MSM_Unlit
  usedShadingModels=Unlit
 }
 nodes [
  n1 = DecalLifetimeOpacity()
  n2 = Clamp(None=LinearInterpolate(A=ScalarParameter() {defaultValue=-3, parameterName=EmissiveFadeOffset}, Alpha=Clamp(None=@n1.Opacity)) {constA=-7})
  n3 = TextureCoordinate()
  n4 = VectorParameter() {parameterName=EmissiveUV, defaultValue=<rgba 2.5 2.5 0 0>}
  n5 = AppendVector(A=@n4.R, B=@n4.R)
  n6 = TextureSampleParameter2D(UVs=Add(A=Multiply(A=@n3, B=@n5), B=AppendVector(A=@n4.B, B=@n4.B))) {parameterName=EmissiveTex, texture=/Game/Global/Textures/Default/Masks_and_Gradients/Cracks_02}
  n7 = Power(Base=Add(A=@n6.R, B=Multiply(A=@n6.B) {constB=0.3}), Exp=ScalarParameter() {defaultValue=1, parameterName=EmissivePower})
  n8 = TextureCoordinate()
  n9 = Constant() {r=0.5}
  n10 = ScalarParameter() {defaultValue=0.5, parameterName=Radius}
  n11 = ScalarParameter() {parameterName=Hardness}
  n12 = LinearInterpolate(A=@n7, B=Multiply(A=@n7, B=SphereMask(A=@n8, B=@n9, Radius=@n10, Hardness=@n11)), Alpha=ScalarParameter() {parameterName=UseSphereMask})
  n13 = TextureSampleParameter2D() {
   parameterName=Emissive Tex Mask
   samplerType=SAMPLERTYPE_LinearColor
   texture=/Game/Effects/Fort_Effects/Gameplay/SmokeDustExplosions/Textures/T_GroundChar_01
  }
  n14 = Multiply(A=Power(Base=@n2), B=Power(Base=Multiply(A=@n12, B=@n13.R), Exp=LinearInterpolate(Alpha=@n2) {constA=14}))
  n15 = Multiply(
   A=LinearInterpolate(
    A=VectorParameter() {group=Emissive, parameterName=EmissiveEdge, defaultValue=<rgba 0.0129 0.0125 0.4 1>}.RGB
    B=VectorParameter() {group=Emissive, parameterName=EmissiveCenter, defaultValue=<rgba 0.293 0.5415 1.25 1>}.RGB
    Alpha=@n14
   )
   B=@n14
  )
  n16 = Clamp(None=@n15)
  MaterialFunctionCall(Emissive=StaticSwitchParameter(
    True=MaterialFunctionCall(LDR Input=@n16, HDR Tint=@n15) {materialFunction=&DeriveHDRfromLDR}.RemappedRange
    False=Constant()
   ) {defaultValue=true, group=Timestamp, parameterName=Use Emissive?}) {materialFunction=&MF_FX_EmissiveAdjustments}
  n17 = ScalarParameter() {
   desc="This parameter drives some fade effects based on the anngle of or decal projection axis and thre wWorld Normal to avoid seeing harsh texture stretch, Use 1 to Enable, 0 To Disable"
   group=Opacity
   parameterName=Decal Projection Fade Enabled
  }
  n18 = SceneTexture() {sceneTextureId=PPI_WorldNormal}
  n19 = Transform(None=Constant3Vector() {constant=<rgba -1 0 0 0>}) {transformSourceType=TRANSFORMSOURCE_Local}
  n20 = Saturate(None=DotProduct(A=ComponentMask(None=@n18.Color) {b=true, g=true, r=true}, B=Normalize(VectorInput=@n19)))
  n21 = ScalarParameter() {
   defaultValue=1
   desc=Projection Fade power value applied by the dot product with the world normal and the decal projection axis.
   group=Opacity
   parameterName=Decal Projection Fade Power
  }
  n22 = Constant() {r=1}
  NamedRerouteDeclaration(None=If(A=@n17, "A > B"=Power(Base=@n20, Exp=@n21) {constExponent=0}, "A == B"=@n22, "A < B"=@n22) {constB=0.99}) {name=Decal Projection Fade, nodeColor=<rgba 0.1953 1 0 1>}
  ScalarParameter() {parameterName=Metallic}
  ScalarParameter() {parameterName=Specular}
  ScalarParameter() {defaultValue=1, parameterName=Roughness}
  n23 = TextureSampleParameter2D() {
   parameterName=DecalTexture
   samplerType=SAMPLERTYPE_LinearColor
   texture=&T_GroundChar_01
  }
  n24 = Multiply(A=@n1.Opacity, B=ScalarParameter() {defaultValue=6, group=Timestamp, parameterName=Fade Transition Multiplier}) {constB=4}
  n25 = Multiply(A=Multiply(A=@n23.R, B=Clamp(None=@n24)), B=ScalarParameter() {defaultValue=1, parameterName=Opacity})
  ShadingPathSwitch(Default=Multiply(A=NamedRerouteUsage(), B=@n25), Forward=@n25, Mobile=@n25)
  VectorParameter() {parameterName=DecalColor, defaultValue=<rgba 0 0 0 1>}
 ]
}
